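# Spectator HUD widget that shows one icon and count for the followed player's build count, wood, stone, metal or gold.
# PreConstruct picks the icon from BuildCountType and zeroes the count outside design time.
# OnBuildCountChanged shows infinity above 999; otherwise it sets the number and, on any difference, even a decrease, plays a scale pop and fires OnBuildCountIncreased.
# Changing the followed player stops all animations.
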WidgetBlueprint /Game/Athena/HUD/Spectator/Inventory/AthenaSpectatorBuildCountWidget
animations [CollectedScale {
  binding_count=1
  length_seconds=0.25
  object=&self:CollectedScale
  sequence {bindings [PlayersLeft {
     id=#1
     display_name=MatCount
     possessed_class=CommonNumericTextBlock
     tracks [{
       type=MovieScene2DTransformTrack
       sections [{
         type=MovieScene2DTransformSection
         end_frame=5
         start_frame=0
         channels [
          Scale.X {type=MovieSceneScriptingFloatChannel, keys [[0,1],[1,1.5],[3,1.5],[4,1]]}
          Scale.Y {type=MovieSceneScriptingFloatChannel, keys [[0,1],[1,1.5],[3,1.5],[4,1]]}
         ]
        }]
      }]
    }]}
 }]
graphs [
 EventGraph {nodes [
   n1 = Event PreConstruct()
   > SetBrushFromSoftTexture!(
    execute
    self:Image Object Reference=Get Icon() -> (Icon:Image Object Reference)
    SoftTexture:Texture 2D Soft Object Reference=Select!(
     BuildCount:Texture 2D Soft Object Reference=Get BuildIcon() -> (BuildIcon:Texture 2D Soft Object Reference)
     Wood:Texture 2D Soft Object Reference=Get WoodIcon() -> (WoodIcon:Texture 2D Soft Object Reference)
     Stone:Texture 2D Soft Object Reference=Get StoneIcon() -> (StoneIcon:Texture 2D Soft Object Reference)
     Metal:Texture 2D Soft Object Reference=Get MetalIcon() -> (MetalIcon:Texture 2D Soft Object Reference)
     Gold:Texture 2D Soft Object Reference=Get GoldIcon() -> (GoldIcon:Texture 2D Soft Object Reference)
     Index:ESpectatorBuildCountType Enum=Get BuildCountType() -> (BuildCountType:ESpectatorBuildCountType Enum=BuildCount)
    ) -> (ReturnValue:Texture 2D Soft Object Reference)
    bMatchSize:bool=false
   ) -> (then)
   > n2 = Branch(Condition=@n1.IsDesignTime)
   SetCurrentValue!(
    execute=@n2.else
    self:Common Numeric Text Block Object Reference=Get MatCount() -> (MatCount:Common Numeric Text Block Object Reference)
    NewValue:float=0
   ) -> (then)
   n3 = Event OnBuildCountChanged!() -> (OutputDelegate, then, OldCount:int=0, NewCount:int=0) {is=title K2Node_Event}
   > n4 = Branch(Condition="integer > integer"(A=@n3.NewCount, B=999) {is=title K2Node_CallFunction})
   > "SetText (Text)"(self=Get MatCount(), InText="INVTEXT(\"∞\")")
   SetCurrentValue(execute=@n4.else, self=Get MatCount(), NewValue="To Float (Integer)"(InInt=@n3.NewCount))
   > Branch(Condition="Not Equal (Integer)"(A=@n3.NewCount, B=@n3.OldCount))
   > PlayAnimation(InAnimation=Get CollectedScale() -> (CollectedScale:Widget Animation Object Reference))
   > Call On Build Count Increased!(execute:Exec, self:Athena Spectator Build Count Widget Object Reference) -> (then:Exec) {is=title K2Node_CallDelegate}
   Event OnFollowedPlayerChanged_BP() {is=title K2Node_Event}
   > StopAllAnimations(self:User Widget Object Reference)
  ]}
 OnBuildCountIncreased {nodes [OnBuildCountIncreased() {is=title K2Node_FunctionEntry}]}
]
variables [
 BuildIcon {type=Texture2D, default=/Game/UI/Icons/Icon-Score/Icon-Score-Building-L}
 GoldIcon {type=Texture2D, default=/Game/UI/Foundation/Textures/Icons/Athena/T-T-Icon-BR-GoldBars-UI-Icon}
 MetalIcon {type=Texture2D, default=/Game/UI/Icons/Icon-Resource/Icon-Resource-Metal-M}
 OnBuildCountIncreased {}
 StoneIcon {type=Texture2D, default=/Game/UI/Icons/Icon-Resource/Icon-Resource-Stone-M}
 WoodIcon {type=Texture2D, default=/Game/UI/Icons/Icon-Resource/Icon-Resource-Wood-M}
]
widgets {
 info {inheritedWidgetCount=0, parentClass=/Script/FortniteUI.AthenaSpectatorBuildCountBase, rootWidgetClass=&SizeBox, widgetCount=4}
 widgets [
  SizeBox_0 {
   widgetClassPath=&SizeBox
   widget=SizeBox_0
   properties {bOverride_MinDesiredWidth=true, maxDesiredWidth=115, minDesiredWidth=115, widthOverride=115}
  }
  HorizontalBox_0 {
   widgetClassPath=&HorizontalBox
   parent=SizeBox_0
   widget=HorizontalBox_0
   properties {renderTransformPivot=<1 0.5>}
   slot {class=SizeBoxSlot, properties {padding=<ltrb 12.5 0 12.5 0>}}
  }
  Icon {
   widgetClassPath=&Image
   bIsVariable=true
   parent=HorizontalBox_0
   widget=Icon
   properties {
    visibility=SelfHitTestInvisible
    brush {
     bIsDynamicallyLoaded=false
     drawAs=Image
     imageType=NoImage
     mirroring=NoMirror
     resourceName=None
     resourceObject=/Game/Athena/HUD/Spectator/Art/SpectatorNameplate-BuildIcon
     tiling=NoTile
     imageSize=<37 37>
     margin=<ltrb 0 0 0 0>
     outlineSettings {
      bUseBrushTransparency=false
      roundingType=HalfHeightRadius
      width=0
      color {colorUseRule=UseColor_Specified, specifiedColor=<rgba 0 0 0 0>}
      cornerRadii=<0 0 0 0>
     }
     tintColor {colorUseRule=UseColor_Specified, specifiedColor=<rgba 1 1 1 1>}
     uVRegion {bIsValid=false, max=<0 0>, min=<0 0>}
    }
   }
   slot {class=HorizontalBoxSlot, properties {horizontalAlignment=HAlign_Center, verticalAlignment=VAlign_Center, padding=<ltrb 1 0 0 0>}}
  }
  MatCount {
   widgetClassPath=/Script/CommonUI.CommonNumericTextBlock
   bIsVariable=true
   parent=HorizontalBox_0
   widget=MatCount
   properties {
    currentNumericValue=888
    justification=Center
    text=888
    font {
     bForceMonospaced=false
     bMaterialIsStencil=false
     fontMaterial=null
     fontObject=&BurbankSmall
     letterSpacing=0
     monospacedWidth=1
     size=20
     skewAmount=0
     typefaceFontName=Bold
     outlineSettings {
      bApplyOutlineToDropShadows=false
      bMiteredCorners=false
      bSeparateFillAlpha=false
      outlineMaterial=null
      outlineSize=0
      outlineColor=<rgba 0 0 0 1>
     }
    }
    renderTransform {angle=0, scale=<1 1>, shear=<0 0>, translation=<0 -1>}
    shadowOffset=<0 0>
   }
   slot {class=HorizontalBoxSlot, properties {verticalAlignment=VAlign_Center, padding=<ltrb 2 0 0 0>, size {sizeRule=Fill, value=1}}}
  }
 ]
}
